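# Pickaxe material for Voyager Remix 1H, with switchable texture or parameter inputs for color, metallic and roughness, and Fresnel-driven color and emissive.
# UseVirtualPlaneGlow dims base color and replaces the emissive with two radial glows on a camera-facing virtual plane around an offset center.
# Panning noise, projected on a local-axis plane, jitters that center for a refraction-like wobble and masks an added normal-mapped Fresnel rim.
# Attributes feed MF_TechArtOverrides_accessories; several leftover nodes are unconnected.
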Material /BRCosmetics/Weapons/FORT_Melee/Pickaxe_VoyagerRemix1H/Materials/FreOutSideGlowable
graph {
 properties {
  bUseMaterialAttributes=true
  bUsedWithMeshDeformer=true
  bUsedWithMeshParticles=true
  bUsedWithNiagaraMeshParticles=true
  bUsedWithSkeletalMesh=true
  pixelDepthOffsetMode=PDOM_Legacy
  usedShadingModels=Default Lit
 }
 nodes [
  ActorPositionWS()
  n1 = CameraVectorWS()
  n2 = CrossProduct(A=@n1, B=Constant3Vector() {constant=<rgba 0 0 1 0>})
  n3 = Add(
   A=Transform(None=VectorParameter() {parameterName=CenterOffset, defaultValue=<rgba 0 100 -30 1>}.RGB) {transformSourceType=TRANSFORMSOURCE_Local}
   B=ObjectPositionWS()
  )
  n4 = Transform(None=VectorParameter() {parameterName=Plane Normal, defaultValue=<rgba 0 0 1 1>}.RGB) {transformSourceType=TRANSFORMSOURCE_Local}
  n5 = Transform(None=VectorParameter() {parameterName=Plane X Axis, defaultValue=<rgba 0 1 0 1>}.RGB) {transformSourceType=TRANSFORMSOURCE_Local}
  n6 = MaterialFunctionCall(Float3=PreSkinnedLocalBounds().Half Extents) {materialFunction=&BreakOutFloat3Components}
  n7 = Max(A=Max(A=@n6.R, B=@n6.R), B=@n6.B)
  n8 = Multiply(A=@n7, B=ScalarParameter() {defaultValue=0.731, parameterName=VirtPlaneScale2})
  n9 = ComponentMask(None=VectorParameter() {parameterName=PanSpeed, defaultValue=<rgba 0 -0.3 0 1>}.RGB) {g=true, r=true}
  n10 = TextureSample(UVs=Panner(
    Coordinate=MaterialFunctionCall!(Plane Normal Axis=@n4, Plane X Axis=@n5, Plane Center=@n3, UV Scale=@n8) {materialFunction=/Game/Athena/Environments/Nexus/Materials/VirtualPlaneCoordinatesAbs}.UVs
    Speed=MaterialFunctionCall(XY=@n9) {materialFunction=&Swizzle}.YX
   )) {samplerType=SAMPLERTYPE_LinearColor, texture=/Game/Packages/Fortress_Sky/Materials/Noise32x322_nearestNeighboor}
  n11 = LinearInterpolate(A=ScalarParameter() {defaultValue=29.8553, parameterName=Refracting}, Alpha=ComponentMask(None=@n10.RGB) {b=true})
  n12 = MaterialFunctionCall(
   Plane Normal Axis=@n1
   Plane X Axis=@n2
   Plane Center=Add(A=@n3, B=@n11)
   UV Scale=Multiply(A=@n7, B=ScalarParameter() {defaultValue=1, parameterName=VirtPlaneScale})
  ) {materialFunction=&VirtualPlaneCoordinatesAbs}
  n13 = Constant2Vector() {g=0.5, r=0.5}
  n14 = Distance(A=@n12.UVs, B=@n13)
  n15 = Power(
   Base=Divide(A=OneMinus(None=@n14), B=ScalarParameter() {defaultValue=1.0302, parameterName=SizeGlow1})
   Exp=ScalarParameter() {defaultValue=29.0186, parameterName=PowerGlow1}
  )
  Add(A=@n15)
  n16 = ScalarParameter() {defaultValue=0.1857, parameterName=BrightnessDiffuse}
  n17 = VectorParameter() {group=FresnelCtrl, parameterName=outside, defaultValue=<rgba 1 0.9393 0 1>}
  n18 = MaterialFunctionCall(
   In=Fresnel(
    ExponentIn=ScalarParameter() {defaultValue=0.7165, group=FresnelCtrl, parameterName=FresnelExpon}
    BaseReflectFractionIn=ScalarParameter() {defaultValue=0.09, group=FresnelCtrl, parameterName=Fresnelref}
   )
   Contrast=ScalarParameter() {defaultValue=0.4116, group=FresnelCtrl, parameterName=FresnelClarEdge}
  ) {materialFunction=&CheapContrast}
  n19 = StaticSwitchParameter(
   True=TextureSampleParameter2D() {parameterName=ColorInput, texture=/Game/Environments/World/Utility/128gray}.RGB
   False=LinearInterpolate(A=VectorParameter() {group=FresnelCtrl, parameterName=inside, defaultValue=<rgba 0.7188 0.3928 0.1098 1>}.RGB, B=@n17.RGB, Alpha=@n18.Result)
  ) {parameterName=ColorTextureCtrl}
  n20 = TextureSampleParameter2D() {parameterName=MetallicInput, texture=&128gray}
  n21 = ScalarParameter() {defaultValue=1, group=BaseValue, parameterName=MetallicMult}
  n22 = TextureSampleParameter2D() {parameterName=RoughnessInput, texture=&128gray}
  n23 = ScalarParameter() {defaultValue=1, group=BaseValue, parameterName=RoughMulti}
  n24 = LinearInterpolate(A=Constant(), B=@n17.RGB, Alpha=@n18.Result)
  n25 = ScalarParameter() {defaultValue=1.8102, group=FresnelCtrl, parameterName=PowerEmissive}
  n26 = Divide(A=OneMinus(None=Distance(A=@n12.UVs, B=@n13)), B=ScalarParameter() {defaultValue=1.0147, parameterName=SizeGlow2})
  n27 = ScalarParameter() {defaultValue=7.5286, parameterName=PowerGlow2}
  n28 = Add(
   A=Multiply(A=@n15, B=VectorParameter() {parameterName=GlowColor1, defaultValue=<rgba 3.0875 5 3.0422 1>}.RGB)
   B=Multiply(A=Power(Base=@n26, Exp=@n27), B=VectorParameter() {parameterName=GlowColor2, defaultValue=<rgba 0.1033 1 0.0772 1>}.RGB)
  )
  n29 = ScalarParameter() {defaultValue=7.8699, parameterName=FresnelEXP}
  n30 = TextureSampleParameter2D() {
   parameterName=InputNorm
   samplerType=SAMPLERTYPE_Normal
   texture=/BRCosmetics/Weapons/FORT_Melee/Pickaxe_VoyagerRemix1H/Textures/T_Pickaxe_Voyager_Remix_N
  }
  n31 = Transform(None=@n30.RGB)
  n32 = Multiply(
   A=VectorParameter() {parameterName=GlowColor3, defaultValue=<rgba 2 2 2 1>}.RGB
   B=Multiply(A=Fresnel(ExponentIn=@n29, Normal=@n31) {baseReflectFraction=0}, B=ComponentMask(None=@n10.RGB) {r=true})
  )
  n33 = SetMaterialAttributes!(
   MaterialAttributes
   Base Color=StaticSwitchParameter(True=Multiply(A=@n16, B=@n19), False=@n19) {parameterName=UseVirtualPlaneGlow}
   Metallic=StaticSwitchParameter(True=Multiply(A=@n20.RGB, B=@n21), False=ScalarParameter() {defaultValue=0.7079, group=BaseValue, parameterName=Metallic}) {parameterName=MetallicTextureCtrl}
   Specular=ScalarParameter() {defaultValue=0.2958, parameterName=Specular}
   Roughness=StaticSwitchParameter(True=Multiply(A=@n22.RGB, B=@n23), False=ScalarParameter() {defaultValue=0.5, group=BaseValue, parameterName=Roughness}) {parameterName=RoughTextureCtrl}
   Emissive Color=StaticSwitchParameter(True=Add(A=@n28, B=@n32), False=Multiply(A=@n24, B=@n25)) {parameterName=UseVirtualPlaneGlow}
   Normal=@n30.RGB
  ) {attributeSetTypes [
    #1
    #2
    #3
    #4
    #5
    #6
   ]}
  MaterialFunctionCall(In=@n33) {materialFunction=&MF_TechArtOverrides_accessories}
  MaterialFunctionCall(Plane Normal Axis=@n4, Plane X Axis=@n5, Plane Center=@n3, UV Scale=@n8) {materialFunction=&VirtualPlaneCoordinates}
  n34 = Normalize(VectorInput=@n2)
  Multiply(A=VertexNormalWS(), B=LinearInterpolate(B=@n11, Alpha=DotProduct(A=@n34, B=@n1)))
  ObjectRadius()
  ObjectRadius()
  ObjectRadius()
 ]
}
